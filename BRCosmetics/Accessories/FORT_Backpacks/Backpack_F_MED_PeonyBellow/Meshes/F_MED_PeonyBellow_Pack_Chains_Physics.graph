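# Physics asset for the chains on the PeonyBellow backpack cosmetic, with capsule bodies along three bone chains and a box body for the pack.
# Chains A and B have seven dynamic links and C has two, each joined by narrow swing and zero-twist limits.
# Each chain's base attaches to the pack with free rotation.
# The last links of A and B also constrain to C's end, so the three chains meet.
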PhysicsAsset /BRCosmetics/Accessories/FORT_Backpacks/Backpack_F_MED_PeonyBellow/Meshes/F_MED_PeonyBellow_Pack_Chains_Physics
bodies {
 base_chainA_1 [{shapeType=Capsule, length=0.5, radius=0.25, shapeName=base_chainA_1_capsule, rotation=<rot 90 0 0>}]
 base_chainB_1 [{shapeType=Capsule, length=0.5, radius=0.25, shapeName=base_chainA_1_capsule, rotation=<rot 90 0 0>}]
 base_chainC_1 [{shapeType=Capsule, length=0.5, radius=0.25, shapeName=base_chainA_1_capsule, rotation=<rot 90 0 0>}]
 dyn_chainA_1 [{shapeType=Capsule, length=1, radius=0.5, shapeName=base_chainA_1_capsule, center=<1 0 0>, rotation=<rot 90 0 0>}]
 dyn_chainA_2 [{shapeType=Capsule, length=1, radius=0.5, shapeName=base_chainA_1_capsule, center=<1 0 0>, rotation=<rot 90 0 0>}]
 dyn_chainA_3 [{shapeType=Capsule, length=1.9889, radius=0.5817, shapeName=base_chainA_1_capsule, center=<1.7775 -0.3712 -1.504>, rotation=<rot 90 0 8.7948>}]
 dyn_chainA_4 [{shapeType=Capsule, length=1, radius=0.5, shapeName=base_chainA_1_capsule, center=<1 0 0>, rotation=<rot 90 0 0>}]
 dyn_chainA_5 [{shapeType=Capsule, length=1.8966, radius=0.5, shapeName=base_chainA_1_capsule, center=<2.0148 0 1.8281>, rotation=<rot 90 0 0>}]
 dyn_chainA_6 [{shapeType=Capsule, length=1, radius=0.5, shapeName=base_chainA_1_capsule, center=<2.1646 0 -1.0161>, rotation=<rot 65.2308 -180 180>}]
 dyn_chainA_7 [{shapeType=Capsule, length=1, radius=0.5, shapeName=base_chainA_1_capsule, center=<2.6837 0 0.14>, rotation=<rot 90 0 0>}]
 dyn_chainB_1 [{shapeType=Capsule, length=1, radius=0.5, shapeName=base_chainA_1_capsule, center=<1 0 0>, rotation=<rot 90 0 0>}]
 dyn_chainB_2 [{shapeType=Capsule, length=1, radius=0.5, shapeName=base_chainA_1_capsule, center=<1 0 0>, rotation=<rot 90 0 0>}]
 dyn_chainB_3 [{shapeType=Capsule, length=1, radius=0.5, shapeName=base_chainA_1_capsule, center=<2.0405 2.1269 0.151>, rotation=<rot 90 0 0>}]
 dyn_chainB_4 [{
   shapeType=Capsule
   length=1.6964
   radius=0.5
   shapeName=base_chainA_1_capsule
   center=<2.3405 -0.3238 -1.7363>
   rotation=<rot 90 -176.4237 -166.1961>
  }]
 dyn_chainB_5 [{shapeType=Capsule, length=1, radius=0.5, shapeName=base_chainA_1_capsule, center=<1 0 0>, rotation=<rot 90 0 0>}]
 dyn_chainB_6 [{shapeType=Capsule, length=1, radius=0.5, shapeName=base_chainA_1_capsule, center=<1 0 0>, rotation=<rot 90 0 0>}]
 dyn_chainB_7 [{
   shapeType=Capsule
   length=1.1956
   radius=0.5
   shapeName=base_chainA_1_capsule
   center=<2.8889 0 -1.8961>
   rotation=<rot 53.9868 -179.9999 -179.9999>
  }]
 dyn_chainC_1 [{shapeType=Capsule, length=1, radius=0.5, shapeName=base_chainA_1_capsule, center=<2.1579 0 0>, rotation=<rot 90 0 0>}]
 dyn_chainC_2 [{shapeType=Capsule, length=1, radius=0.5, shapeName=base_chainA_1_capsule, center=<1 0 0>, rotation=<rot 90 0 0>}]
 pack [{
   shapeType=Box
   extentX=3.9302
   extentY=12.3643
   extentZ=25.1476
   shapeName=pack_box
   center=<-0.5478 -3.8951 -4.5606>
   rotation=<rot -7.6838 -90.112 -68.7671>
  }]
}
constraints [
 {bone1Name=dyn_chainA_1, bone2Name=base_chainA_1, swing1LimitDegrees=10, swing2LimitDegrees=10, twistLimitDegrees=0}
 {bone1Name=dyn_chainA_2, bone2Name=dyn_chainA_1, swing1LimitDegrees=10, swing2LimitDegrees=10, twistLimitDegrees=0}
 {bone1Name=dyn_chainA_3, bone2Name=dyn_chainA_2, swing1LimitDegrees=10, swing2LimitDegrees=10, twistLimitDegrees=0}
 {bone1Name=dyn_chainA_4, bone2Name=dyn_chainA_3, swing1LimitDegrees=10, swing2LimitDegrees=10, twistLimitDegrees=0}
 {bone1Name=dyn_chainA_5, bone2Name=dyn_chainA_4, swing1LimitDegrees=10, swing2LimitDegrees=10, twistLimitDegrees=0}
 {bone1Name=dyn_chainA_6, bone2Name=dyn_chainA_5, swing1LimitDegrees=10, swing2LimitDegrees=10, twistLimitDegrees=0}
 {bone1Name=dyn_chainA_7, bone2Name=dyn_chainA_6, swing1LimitDegrees=10, swing2LimitDegrees=10, twistLimitDegrees=0}
 {bone1Name=dyn_chainB_1, bone2Name=base_chainB_1, swing1LimitDegrees=3, swing2LimitDegrees=3, twistLimitDegrees=0}
 {bone1Name=dyn_chainB_2, bone2Name=dyn_chainB_1, swing1LimitDegrees=7, swing2LimitDegrees=7, twistLimitDegrees=0}
 {bone1Name=dyn_chainB_3, bone2Name=dyn_chainB_2, swing1LimitDegrees=7, swing2LimitDegrees=7, twistLimitDegrees=0}
 {bone1Name=dyn_chainB_4, bone2Name=dyn_chainB_3, swing1LimitDegrees=7, swing2LimitDegrees=7, twistLimitDegrees=0}
 {bone1Name=dyn_chainB_5, bone2Name=dyn_chainB_4, swing1LimitDegrees=7, swing2LimitDegrees=7, twistLimitDegrees=0}
 {bone1Name=dyn_chainB_6, bone2Name=dyn_chainB_5, swing1LimitDegrees=7, swing2LimitDegrees=7, twistLimitDegrees=0}
 {bone1Name=dyn_chainB_7, bone2Name=dyn_chainB_6, swing1LimitDegrees=7, swing2LimitDegrees=7, twistLimitDegrees=0}
 {bone1Name=dyn_chainC_1, bone2Name=base_chainC_1, swing1LimitDegrees=4, swing2LimitDegrees=4, twistLimitDegrees=0}
 {bone1Name=dyn_chainC_2, bone2Name=dyn_chainC_1, swing1LimitDegrees=2, swing2LimitDegrees=2, twistLimitDegrees=0}
 {bone1Name=dyn_chainB_7, bone2Name=dyn_chainC_2, swing1LimitDegrees=2, swing2LimitDegrees=2, twistLimitDegrees=0}
 {bone1Name=dyn_chainA_7, bone2Name=dyn_chainC_2, swing1LimitDegrees=2, swing2LimitDegrees=2, twistLimitDegrees=0}
 {
  bone1Name=base_chainC_1
  bone2Name=pack
  swing1LimitDegrees=45
  swing1Motion=Free
  swing2LimitDegrees=45
  swing2Motion=Free
  twistLimitDegrees=45
  twistMotion=Free
 }
 {
  bone1Name=base_chainA_1
  bone2Name=pack
  swing1LimitDegrees=45
  swing1Motion=Free
  swing2LimitDegrees=45
  swing2Motion=Free
  twistLimitDegrees=45
  twistMotion=Free
 }
 {
  bone1Name=base_chainB_1
  bone2Name=pack
  swing1LimitDegrees=45
  swing1Motion=Free
  swing2LimitDegrees=45
  swing2Motion=Free
  twistLimitDegrees=45
  twistMotion=Free
 }
]
