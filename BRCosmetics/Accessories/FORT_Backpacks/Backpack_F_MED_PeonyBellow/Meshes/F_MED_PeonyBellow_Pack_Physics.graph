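# Physics asset for the PeonyBellow backpack cosmetic, defining a dangling charm chain on the pack mesh.
# A small capsule on base_charm_1 anchors five larger, identical capsules on dyn_charm_1 through dyn_charm_5.
# Constraints link each bone to the previous one, with zero twist range and narrow swing limits.
# Swing is asymmetric and looser in the first two joints, then tighter and uniform toward the tip.
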PhysicsAsset /BRCosmetics/Accessories/FORT_Backpacks/Backpack_F_MED_PeonyBellow/Meshes/F_MED_PeonyBellow_Pack_Physics
bodies {
 base_charm_1 [{shapeType=Capsule, length=0.5, radius=0.25, shapeName=base_chainA_1_capsule, rotation=<rot 90 0 0>}]
 dyn_charm_1 [{shapeType=Capsule, length=1, radius=0.5, shapeName=base_chainA_1_capsule, center=<1 0 0>, rotation=<rot 90 0 0>}]
 dyn_charm_2 [{shapeType=Capsule, length=1, radius=0.5, shapeName=base_chainA_1_capsule, center=<1 0 0>, rotation=<rot 90 0 0>}]
 dyn_charm_3 [{shapeType=Capsule, length=1, radius=0.5, shapeName=base_chainA_1_capsule, center=<1 0 0>, rotation=<rot 90 0 0>}]
 dyn_charm_4 [{shapeType=Capsule, length=1, radius=0.5, shapeName=base_chainA_1_capsule, center=<1 0 0>, rotation=<rot 90 0 0>}]
 dyn_charm_5 [{shapeType=Capsule, length=1, radius=0.5, shapeName=base_chainA_1_capsule, center=<1 0 0>, rotation=<rot 90 0 0>}]
}
constraints [
 {bone1Name=dyn_charm_1, bone2Name=base_charm_1, swing1LimitDegrees=28, swing2LimitDegrees=10, twistLimitDegrees=0}
 {bone1Name=dyn_charm_2, bone2Name=dyn_charm_1, swing1LimitDegrees=28, swing2LimitDegrees=16, twistLimitDegrees=0}
 {bone1Name=dyn_charm_3, bone2Name=dyn_charm_2, swing1LimitDegrees=16, swing2LimitDegrees=16, twistLimitDegrees=0}
 {bone1Name=dyn_charm_4, bone2Name=dyn_charm_3, swing1LimitDegrees=16, swing2LimitDegrees=16, twistLimitDegrees=0}
 {bone1Name=dyn_charm_5, bone2Name=dyn_charm_4, swing1LimitDegrees=16, swing2LimitDegrees=16, twistLimitDegrees=0}
]
